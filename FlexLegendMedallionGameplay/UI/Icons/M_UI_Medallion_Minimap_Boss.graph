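# Translucent UI material that draws a minimap medallion icon with a boss indicator overlaid on it.
# The MedallionIcon texture parameter supplies the base, defaulting to an empty image.
# BossIndicator samples scaled, offset normalized UI UVs with clamped addressing, shrinking the badge into the lower-right corner.
# The indicator's RGB serves as its own blend mask over the icon color, and the larger of the two alphas forms the combined alpha.
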Material /FlexLegendMedallionGameplay/UI/Icons/M_UI_Medallion_Minimap_Boss
graph {
 properties {bAutomaticallySetUsageInEditor=false, blendMode=BLEND_Translucent, materialDomain=MD_UI, usedShadingModels=Default Lit}
 nodes [
  n1 = TextureSampleParameter2D() {parameterName=MedallionIcon, texture=/Game/Athena/UI/Frontend/Spatial/Art/T_Ui_Empty}
  n2 = MaterialFunctionCall() {materialFunction=&GetUserInterfaceUV}
  n3 = TextureSampleParameter2D(UVs=Add(A=Multiply(A=@n2.Normalized UV) {constB=1.8}) {constB=-0.8}) {parameterName=BossIndicator, samplerSource=SSM_Clamp_WorldGroupSettings, texture=&/T_UI_Medallion_Boss}
  LinearInterpolate(A=@n1.RGB, B=@n3.RGB, Alpha=@n3.RGB)
  Max(A=@n1.A, B=@n3.A)
 ]
}
